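# Unlit masked material for Niagara sprite confetti on the GelatoCourage backpack cosmetic.
# A 4x4 SubUV confetti normal map drives a reflection vector that samples a desaturated glass cubemap, faking shiny flakes without lighting.
# That reflection is tinted by particle color, darkened by the normal's green channel, then fed through MF_FX_EmissiveAdjustments.
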Material /BRCosmetics/Accessories/FORT_Backpacks/Backpack_M_MED_GelatoCourage/FX/M_Backpack_GelatoCourage_Confettis
graph {
 properties {bUsedWithNiagaraSprites=true, blendMode=BLEND_Masked, pixelDepthOffsetMode=PDOM_Legacy, shadingModel=MSM_Unlit, usedShadingModels=Unlit}
 nodes [
  n1 = TextureSample() {samplerType=SAMPLERTYPE_Normal, texture=/Game/Effects/Fort_Effects/Textures/Confetti/T_Confetti_SubUV_4X4_03}
  n2 = ReflectionVectorWS(CustomWorldNormal=Transform(None=@n1.RGB))
  n3 = OneMinus(None=Multiply(A=@n1.G) {constB=0.5})
  n4 = ParticleColor()
  MaterialFunctionCall(Emissive=Multiply(
    A=TextureSample(UVs=@n2) {samplerType=SAMPLERTYPE_LinearColor, texture=/Game/Creative/Environments/Meshes/Glass/Textures/T_Glass_Cubemap_01_Desaturated}.RGB
    B=Multiply(A=@n3, B=@n4.RGB)
   )) {materialFunction=&MF_FX_EmissiveAdjustments}
  TextureSample() {texture=/Game/Effects/Fort_Effects/Textures/Confetti/T_Confetti_SubUV_4X4_A}
  VectorParameter() {parameterName=Orange, defaultValue=<rgba 1 0.4 0 1>}
 ]
}
