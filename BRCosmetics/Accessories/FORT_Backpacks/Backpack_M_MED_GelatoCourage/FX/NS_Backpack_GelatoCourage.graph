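# Niagara effect for the GelatoCourage back bling that plays a gold transition with a celebratory finish, plus ambient sparkles.
# Sparks climb a cylinder alongside a rising light, while GPU soft sparkles drift from the skeletal mesh.
# User.GoldTransitionCompleteTime delays confetti squares, streamers, flashing mesh sparkles and one burst of the two-group Flare002.
# Time-of-day scalars tune brightness, front-end modifiers adjust flashing sparkles, and the older Flare emitter is disabled.
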NiagaraSystem /BRCosmetics/Accessories/FORT_Backpacks/Backpack_M_MED_GelatoCourage/FX/NS_Backpack_GelatoCourage
dependencies {
 usedDataInterfaces [
  &NiagaraDataInterfaceArrayInt32
  &NiagaraDataInterfaceCurve
  &NiagaraDataInterfaceSpriteRendererInfo
  &NiagaraDataInterfaceSkeletalMesh
 ]
 usedDynamicInputs []
 usedModules [
  &SystemState
  &self:SystemSpawnScript.SystemScriptSource.SystemScriptGraph.NiagaraNodeAssignment_0.SetVariables_#1
  &self:SystemSpawnScript.SystemScriptSource.SystemScriptGraph.NiagaraNodeAssignment_1.SetVariables_#2
  &self:SystemSpawnScript.SystemScriptSource.SystemScriptGraph.NiagaraNodeAssignment_2.SetVariables_#3
  &self:SystemSpawnScript.SystemScriptSource.SystemScriptGraph.NiagaraNodeAssignment_3.SetVariables_#4
 ]
}
emitters {
 ConfettiSquares {
  emitterUpdateScript [
   &EmitterState(
    Life Cycle Mode=::NewEnumerator1
    Loop Behavior=::NewEnumerator1
    Loop Delay=$User.GoldTransitionCompleteTime
    Loop Duration=1
    UseLoopDelay=true
   )
   &SpawnBurst_Instantaneous(Spawn Count=25, Use Spawn Probability=false, UseLoopCountLimit=false)
  ]
  particleSpawnScript [
   &InitializeParticle(
    Color=<rgba 2 1.5251 0.324 1>
    Lifetime Max=3.33
    Lifetime Mode=::NewEnumerator1
    Position Offset:NiagaraPosition=$User.PositionOffset
    Position Offset Coordinate Space=::Local
    Sprite Rotation Angle=0.125
    Sprite Rotation Mode=::NewEnumerator1
    Sprite Size Mode=::NewEnumerator1
    Uniform Sprite Size Max=6
    Uniform Sprite Size Min=3
    UsePositionOffset=true
    Write Material Random=true
   )
   &AddVelocity(
    Cone Angle=100
    Cone Axis=<0 0 1>
    Module.UseConeFalloff=true
    Module.UseVelDistribution=true
    Override Seed=false
    Velocity Mode=::NewEnumerator2
    Velocity Speed=&RandomRangeFloat(Minimum=50, Maximum=400)
   )
  ]
  particleUpdateScript [
   &ParticleState()
   &GravityForce(Gravity=<0 0 -111>)
   /Niagara/Modules/Update/Forces/CurlNoiseForce(Noise Frequency=0.02, Noise Strength=100, Pan Noise=false)
   &Drag(
    Drag=&FloatFromCurve(FloatCurve=DataInterface() {propertyValues {Curve {keys [<at 0 0.5>,<at 1 1>]}}}, Scale Curve=8)
    Use Linear Drag=true
    Use Rotational Drag=true
   )
   &InheritSourceMovement(
    Applied Position Scale=&VectorFromFloat(Value=&FloatFromCurve(FloatCurve=DataInterface() {propertyValues {Curve {keys [<at 0.2 1>,<at 0.8 0>]}}}))
    ApplyForce=false
    ApplyPosition=true
    ApplyPositionFalloff=false
   )
   &SolveForcesAndVelocity(Clamp Velocity=false, Limit Acceleration=false)
   &ScaleSpriteSize(Scale Factor=&Vector2DFromFloat(Value=&FloatFromCurve(
      FloatCurve=DataInterface() {propertyValues {Curve {keys [
          <at 0 0 RCIM_Cubic>
          {arriveTangent=-0.0623, interpMode=RCIM_Cubic, leaveTangent=-0.0623, tangentMode=RCTM_User, time=0.04, value=1}
          {arriveTangent=-0.0083, interpMode=RCIM_Cubic, leaveTangent=-0.0083, tangentMode=RCTM_User, time=0.5, value=1}
          <at 1 0 RCIM_Cubic>
         ]}}}
      Scale Curve=0.7
     )))
   &SubUVAnimation(
    "Sprite Renderer "=DataInterface() {propertyValues {SpriteRenderer=&self:ConfettiSquares_0.NiagaraSpriteRendererProperties_0}}
    Start Frame Offset:NiagaraInt32=&RandomRangeInt(Minimum=0, Maximum=555)
    UseEndFrame=false
    UseStartFrame=false
   )
   &ScaleColor(
    Scale RGB=&VectorFromFloat(Value=&Multiply_Float(
      A=&FloatFromCurve(FloatCurve=DataInterface() {propertyValues {Curve {keys [
           {arriveTangent=-270.6918, interpMode=RCIM_Cubic, leaveTangent=-270.6918, tangentMode=RCTM_User, time=0.1, value=40}
           {interpMode=RCIM_Cubic, tangentMode=RCTM_User, time=0.4, value=1}
           {interpMode=RCIM_Cubic, tangentMode=RCTM_User, time=1, value=0}
          ]}}})
      B=&FNTimeOfDayScalar(Night Scale=0.2, Morning Scale=0.333, Day Scale=0.5, Evening Scale=0.25)
     ))
    ScaleA=true
    ScaleRGB=true
   )
   /SetVariables(Particles.randomCurve:NiagaraFloat=&FloatFromCurve(
     FloatCurve=DataInterface() {propertyValues {Curve {keys [<at 0 0>,<at 0.1 1>,<at 0.15 0>]}}}
     CurveIndex=&Subtract_Float(A=$Particles.NormalizedAge, B=&ReturnNormalizedExecIndex(Normalized Index Scale=3))
    )) {isSetParameters=true}
   /SetVariables(
    Particles.SparkCameraOffset=20
    Particles.SparkColor:LinearColor=&MultiplyLinearColor(
     Linear Color=<rgba 4 2.2987 0.48 40>
     Scale Factor=&MakeVector4_FromFloat(Float=$Particles.randomCurve)
    )
    Particles.SparkSize:Vector2f=&Vector2DFromFloat(Value=&Multiply_Float(A=&RandomRangeFloat(Minimum=10, Maximum=20), B=$Particles.randomCurve))
   ) {isSetParameters=true}
   /SetVariables(
    Particles.GlowColor=&MultiplyLinearColor(
     Linear Color=<rgba 0.2 0.1058 0.0287 1>
     Scale Factor=&MakeVector4_FromFloat(Float=&FloatFromCurve(
       FloatCurve=DataInterface() {propertyValues {Curve {keys [{interpMode=RCIM_Cubic, tangentMode=RCTM_User, time=0.3, value=1}, {interpMode=RCIM_Cubic, tangentMode=RCTM_User, time=1, value=0}]}}}
       Scale Curve=&Multiply_Float(A=1, B=$System.TOD)
      ))
    )
    Particles.GlowSize=&Multiply_Vector2D(A=$Particles.SpriteSize, B=&Multiply_Vector2D_ByFloat(Float=2, Vector2D=<1 1>))
   ) {isSetParameters=true}
  ]
  rendererClasses [&NiagaraSpriteRendererProperties]
  renderers [
   {rendererClass=&NiagaraSpriteRendererProperties}
   {rendererClass=&NiagaraSpriteRendererProperties, rendererIndex=1}
   {rendererClass=&NiagaraSpriteRendererProperties, rendererIndex=2}
  ]
 }
 ConfettiStreamers {
  emitterUpdateScript [
   &EmitterState(
    Life Cycle Mode=::NewEnumerator1
    Loop Behavior=::NewEnumerator1
    Loop Delay=$User.GoldTransitionCompleteTime
    Loop Duration=1
    UseLoopDelay=true
   )
   &SpawnBurst_Instantaneous(
    Spawn Count=&RandomRangeInt(Minimum=4, Maximum=6)
    Use Spawn Probability=false
    UseLoopCountLimit=false
   )
   /SetVariables(Emitter.random:NiagaraInt32=&SelectIntFromArray(
     Direct Array Index=&RandomRangeInt(Minimum=0, Maximum=3)
     Int Selection Array=DataInterface() {propertyValues {IntData [2,4,6,9]}}
    )) {isSetParameters=true}
  ]
  particleSpawnScript [
   &InitializeParticle(
    Color=&LerpLinearColors(
     StartColor=<rgba 1 0.6453 0.0746 1>
     EndColor=&SelectLinearColorFromArray(Array Sampling Mode=::NewEnumerator0, Direct Array Index=?, Color Selection Array=$User.ConfettiColorArray, Random Seed=0)
     LerpFactor=0
    )
    Lifetime=&RandomRangeFloat(Minimum=0.8, Maximum=2)
    Mesh Scale=&VectorFromFloat(Value=&RandomRangeFloat(Minimum=0.25, Maximum=0.3))
    Mesh Scale Mode=::NewEnumerator4
    Position Offset:NiagaraPosition=$User.PositionOffset
    UsePositionOffset=true
    Write Material Random=true
   )
   &AddVelocity(
    Cone Angle=90
    Cone Axis=<0 0 1>
    Distribution Along Cone Axis=0.7
    Inner Cone Angle=60
    Module.UseConeFalloff=true
    Module.UseVelDistribution=true
    Override Seed=true
    Random Seed:NiagaraInt32=$Emitter.random
    Randomness Mode:ENiagaraRandomnessMode=::NewEnumerator1
    Velocity Mode=::NewEnumerator2
    Velocity Speed=&Multiply_Float(
     A=&RandomRangeFloat(Minimum=44, Maximum=88, Randomness Mode=::NewEnumerator1, Random Seed=$Emitter.random)
     B=10
    )
   )
   &InitialMeshOrientation(
    Mesh Orientation Mode=::NewEnumerator2
    Orientation Coordinate Space=::Local
    Orientation Vector=$Particles.Velocity
    Rotation=&MakeVector(X=&ReturnNormalizedExecIndex(), Y=0, Z=0)
    Use Rotation Vector=true
   )
  ]
  particleUpdateScript [
   &ParticleState()
   &GravityForce(Gravity=<0 0 -200>)
   &Drag(Drag=400, Use Linear Drag=true, Use Rotational Drag=false)
   &SolveForcesAndVelocity(Clamp Velocity=false, Limit Acceleration=false)
   &ScaleColor(
    Scale Alpha=0
    Scale RGB=&VectorFromFloat(Value=&FNTimeOfDayScalar(Night Scale=0.17, Morning Scale=0.333, Day Scale=0.666, Evening Scale=0.333))
    ScaleA=true
    ScaleRGB=true
   )
   &DynamicMaterialParameters(
    Index 0 Param 1=&RandomRangeFloat(Minimum=0.8, Maximum=1.2)
    Module.Index0Param0Write=true
    Module.Index0Param1Write=true
    Module.Index0Param2Write=true
    Module.Index0Param3Write=true
   )
  ]
  rendererClasses [&NiagaraMeshRendererProperties]
  renderers [{rendererClass=&NiagaraMeshRendererProperties, rendererIndex=0}]
 }
 Flare {
  bEnabled=false
  emitterUpdateScript [
   &EmitterState(Life Cycle Mode=::NewEnumerator1, Loop Behavior=::NewEnumerator1, Loop Duration=1, UseLoopDelay=false)
   &SpawnBurst_Instantaneous(Use Spawn Probability=false, UseLoopCountLimit=false)
  ]
  particleSpawnScript [
   &InitializeParticle(
    Color=<rgba 3 1.5693 0.4564 1>
    Lifetime=0.8
    Position Offset:NiagaraPosition=$User.PositionOffset
    Sprite Size Mode=::NewEnumerator3
    Uniform Sprite Size=25.2211
    UsePositionOffset=true
    Write Material Random=true
   )
   &CameraOffset(Camera Offset Amount=&FN_FrontEndFloatModifier(Gameplay=30, Front End=0))
  ]
  particleUpdateScript [
   &ParticleState()
   &ScaleSpriteSize(Uniform Curve Sprite Scale=DataInterface() {propertyValues {Curve {keys [
        {arriveTangent=5.6485, interpMode=RCIM_Cubic, leaveTangent=5.6485, tangentMode=RCTM_User, time=0, value=0}
        {arriveTangent=-0.1641, interpMode=RCIM_Cubic, leaveTangent=-0.1641, tangentMode=RCTM_User, time=0.5, value=1}
       ]}}})
   &ScaleColor(
    Scale RGB=&Multiply_VectorByFloat(
     Vector=<1 1 1>
     Float=&FloatFromCurve(FloatCurve=DataInterface() {propertyValues {Curve {keys [
          {arriveTangent=9.0725, interpMode=RCIM_Cubic, leaveTangent=9.0725, tangentMode=RCTM_User, time=0, value=0}
          {arriveTangent=-0.2385, leaveTangent=-0.2385, time=0.1408, value=0.9847}
          {arriveTangent=-0.0284, interpMode=RCIM_Cubic, leaveTangent=-0.0284, tangentMode=RCTM_User, time=0.3, value=1.025}
          {interpMode=RCIM_Cubic, leaveTangent=-0.0405, time=1, value=0}
         ]}}})
    )
    ScaleA=true
    ScaleRGB=true
   )
   /SetVariables(Particles.Flaresize2:Vector2f=&Multiply_Vector2D(A=$Particles.SpriteSize, B=<1.5149 0.25>)) {enabled=false, isSetParameters=true, module=SetVariables_#5}
   &FN_Light_Attributes(Create Exponent Attribute=true, Create Radius Attribute=true, Exponent=4, Radius=50)
  ]
  rendererClasses [&NiagaraSpriteRendererProperties, &NiagaraLightRendererProperties]
  renderers [
   {rendererClass=&NiagaraSpriteRendererProperties}
   {rendererClass=&NiagaraLightRendererProperties, rendererIndex=1}
  ]
 }
 Flare002 {
  emitterUpdateScript [
   &EmitterState(Inactive Response=::NewEnumerator1, Life Cycle Mode=::NewEnumerator1, Loop Behavior=::NewEnumerator1, Loop Duration=2.5, UseLoopDelay=true)
   &SpawnBurst_Instantaneous(Spawn Time=$User.GoldTransitionCompleteTime, Use Spawn Probability=false, UseLoopCountLimit=false)
   &SpawnBurst_Instantaneous(Spawn Group=1, Use Spawn Probability=false, UseLoopCountLimit=false) {module=SpawnBurst_Instantaneous001}
  ]
  particleSpawnScript [
   /SetVariables(Particles.VisibilityTag=&SpawnGroup()) {isSetParameters=true}
   &InitializeParticle(
    Color=<rgba 2 1.2087 0.3043 1>
    Lifetime=&MaskFloatBySpawnGroup(Masked Float=1.8, Passthrough Float=0.8)
    Position Offset=&AddVectorToPosition(
     Position=$User.PositionOffset
     Vector=&MaskVectorBySpawnGroup(Masked Vector=<0 0 20>, Passthrough Vector=<0 0 0>)
    )
    Position Offset Coordinate Space=::Local
    Sprite Size Mode=::NewEnumerator3
    Uniform Sprite Size=76.8597
    UsePositionOffset=true
    Write Material Random=true
   )
   &CameraOffset()
  ]
  particleUpdateScript [
   &ParticleState()
   &ScaleSpriteSize(Uniform Curve Sprite Scale=DataInterface() {propertyValues {Curve {keys [
        {arriveTangent=2.1817, interpMode=RCIM_Cubic, leaveTangent=2.1817, tangentMode=RCTM_User, time=0, value=0.4}
        {interpMode=RCIM_Cubic, leaveTangent=-0.1641, tangentMode=RCTM_User, time=0.7, value=1}
       ]}}})
   &ScaleColor(
    Scale Alpha=&MakeFloatFromVector(
     Vector=&Multiply_VectorByFloat(
      Vector=<1 1 1>
      Float=&FloatFromCurve(
       FloatCurve=DataInterface() {propertyValues {Curve {keys [
           {arriveTangent=9.0725, interpMode=RCIM_Cubic, leaveTangent=9.0725, tangentMode=RCTM_User, time=0, value=0}
           {arriveTangent=-0.2385, leaveTangent=-0.2385, time=0.1408, value=0.9847}
           {arriveTangent=-0.0284, interpMode=RCIM_Cubic, leaveTangent=-0.0284, tangentMode=RCTM_User, time=0.3, value=1.025}
           {interpMode=RCIM_Cubic, leaveTangent=-0.0405, time=1, value=0}
          ]}}}
       CurveIndex=$System.DelayTimer
      )
     )
     Channel=::NewEnumerator0
    )
    Scale RGB=&Multiply_VectorByFloat(
     Vector=<1 1 1>
     Float=&MaskFloatBySpawnGroup(
      Masked Float=$System.SystemCurve
      Passthrough Float=&FloatFromCurve(FloatCurve=DataInterface() {propertyValues {Curve {keys [
           {arriveTangent=9.0725, interpMode=RCIM_Cubic, leaveTangent=9.0725, tangentMode=RCTM_User, time=0, value=0}
           {arriveTangent=-0.2385, leaveTangent=-0.2385, time=0.1408, value=0.9847}
           {arriveTangent=-0.0284, interpMode=RCIM_Cubic, leaveTangent=-0.0284, tangentMode=RCTM_User, time=0.3, value=1.025}
           {interpMode=RCIM_Cubic, leaveTangent=-0.0405, time=1, value=0}
          ]}}})
     )
    )
    ScaleA=false
    ScaleRGB=true
   )
   /SetVariables(
    Particles.FlareColor=&MultiplyLinearColor(
     Linear Color=<rgba 4 1.771 0 400020>
     Scale Factor=&MakeVector4_FromFloat(Float=&Multiply_Float(A=$System.SystemCurve, B=$System.TOD))
    )
    Particles.Flaresize2:Vector2f=&Multiply_Vector2D(A=$Particles.SpriteSize, B=&Multiply_Vector2D_ByFloat(Float=2, Vector2D=<1 1>))
   ) {isSetParameters=true, module=SetVariables_#5}
   /SetVariables(Particles.MeshColor:LinearColor=&MultiplyLinearColor(
     Linear Color=<rgba 4 2.4208 0.3557 1>
     Scale Factor=&MakeVector4_FromFloat(Float=&Multiply_Float(A=$System.SystemCurve, B=$System.TOD))
    )) {isSetParameters=true}
   &FN_Light_Attributes(
    Color=$Particles.Color
    Create Color Attribute=true
    Create Exponent Attribute=true
    Create Radius Attribute=true
    Exponent=8
    Radius=40
    Scale Color=10
   )
   &ScaleColor(Scale Alpha=$System.TOD, Scale RGB=&VectorFromFloat(Value=1), ScaleA=true, ScaleRGB=true) {module=ScaleColor001}
  ]
  rendererClasses [
   &NiagaraSpriteRendererProperties
   &NiagaraMeshRendererProperties
   &NiagaraLightRendererProperties
  ]
  renderers [
   {rendererClass=&NiagaraSpriteRendererProperties}
   {rendererClass=&NiagaraMeshRendererProperties, rendererIndex=1}
   {rendererClass=&NiagaraSpriteRendererProperties, rendererIndex=2}
   {rendererClass=&NiagaraLightRendererProperties, rendererIndex=3}
   {rendererClass=&NiagaraMeshRendererProperties, rendererIndex=4}
   {rendererClass=&NiagaraSpriteRendererProperties, rendererIndex=5}
  ]
 }
 FlashingSparkles {
  emitterUpdateScript [
   &EmitterState(
    Delay First Loop Only=true
    Inactive Response=::NewEnumerator1
    Life Cycle Mode=::NewEnumerator1
    Loop Delay=$User.GoldTransitionCompleteTime
    Loop Duration=&RandomRangeFloat(Minimum=1, Maximum=0.5)
    UseLoopDelay=true
   )
   &SpawnBurst_Instantaneous(Use Spawn Probability=false, UseLoopCountLimit=false)
   &SpawnRate(SpawnRate=2, Use Spawn Probability=false)
  ]
  particleSpawnScript [
   &InitializeParticle(
    Color=<rgba 0.5 0.2833 0 1>
    Lifetime Max=0.15
    Lifetime Min=0.2
    Lifetime Mode=::NewEnumerator1
    Sprite Rotation Angle=&UniformAOrBFloat(A=0, B=0.125)
    Sprite Rotation Mode=::NewEnumerator1
    Sprite Size Mode=::NewEnumerator1
    Uniform Sprite Size Max=66
    Uniform Sprite Size Min=33
    UsePositionOffset=false
    Write Material Random=true
   )
   &SkeletalMeshLocation(
    MaskSpawnGroup=false
    Mesh Sampling Type=::NewEnumerator8
    OffsetAlongNormal=false
    OverrideRandomSeed=false
    Skeletal Mesh=$System.Skeletal Mesh
   )
   &CameraOffset(Camera Offset Amount=&FN_FrontEndFloatModifier(Gameplay=150, Front End=25))
  ]
  particleUpdateScript [
   &ParticleState()
   &ScaleColor(
    Scale Alpha=&FloatFromCurve(FloatCurve=DataInterface() {propertyValues {Curve {keys [
         {arriveTangent=9.8157, interpMode=RCIM_Cubic, leaveTangent=9.8157, tangentMode=RCTM_User, time=0, value=0}
         {arriveTangent=-1.4617, interpMode=RCIM_Cubic, leaveTangent=-1.4617, tangentMode=RCTM_User, time=0.2, value=1}
         {interpMode=RCIM_Cubic, tangentMode=RCTM_User, time=0.3, value=0}
         {interpMode=RCIM_Cubic, tangentMode=RCTM_User, time=0.5, value=0.75}
         {interpMode=RCIM_Cubic, tangentMode=RCTM_Break, time=1, value=0}
        ]}}})
    ScaleA=true
    ScaleRGB=true
   )
   &ScaleSpriteSize(
    Uniform Curve Scale=&FN_FrontEndFloatModifier(Gameplay=0.2, Front End=0.3)
    Uniform Curve Sprite Scale=DataInterface() {propertyValues {Curve {keys [{tangentMode=RCTM_User, time=0, value=1}, {tangentMode=RCTM_User, time=1, value=0}]}}}
   )
  ]
  rendererClasses [&NiagaraSpriteRendererProperties]
  renderers [{rendererClass=&NiagaraSpriteRendererProperties}]
 }
 Light {
  emitterUpdateScript [
   &EmitterState(Inactive Response=::NewEnumerator1, Life Cycle Mode=::NewEnumerator1, Loop Behavior=::NewEnumerator1, Loop Duration=2, UseLoopDelay=true)
   &SpawnBurst_Instantaneous(Use Spawn Probability=false, UseLoopCountLimit=false)
  ]
  particleSpawnScript [
   &InitializeParticle(
    Color=<rgba 3 0.9876 0.2312 1>
    Position Offset:NiagaraPosition=$User.PositionOffset
    Position Offset Coordinate Space=::Local
    Sprite Size Mode=::NewEnumerator3
    Uniform Sprite Size=25.2211
    UsePositionOffset=true
    Write Material Random=true
   )
   &CameraOffset()
  ]
  particleUpdateScript [
   &ParticleState(Kill Particles When Lifetime Has Elapsed=false)
   &ScaleSpriteSize(Uniform Curve Sprite Scale=DataInterface() {propertyValues {Curve {keys [
        {arriveTangent=5.6485, interpMode=RCIM_Cubic, leaveTangent=5.6485, tangentMode=RCTM_User, time=0, value=0}
        {arriveTangent=-0.1641, interpMode=RCIM_Cubic, leaveTangent=-0.1641, tangentMode=RCTM_User, time=0.5, value=1}
       ]}}})
   &ScaleColor(
    Scale RGB=&Multiply_VectorByFloat(
     Vector=<1 1 1>
     Float=&FloatFromCurve(
      FloatCurve=DataInterface() {propertyValues {Curve {keys [
          {arriveTangent=9.0725, interpMode=RCIM_Cubic, leaveTangent=9.0725, tangentMode=RCTM_User, time=0, value=0}
          {arriveTangent=-0.2385, leaveTangent=-0.2385, time=0.1408, value=0.9847}
          {arriveTangent=-0.0284, interpMode=RCIM_Cubic, leaveTangent=-0.0284, tangentMode=RCTM_User, time=0.3, value=1.025}
          {arriveTangent=-0.5694, interpMode=RCIM_Cubic, leaveTangent=-0.5694, time=1.5, value=1}
          <at 2.1 0>
         ]}}}
      CurveIndex=$Emitter.Age
     )
    )
    ScaleA=true
    ScaleRGB=true
   )
   &FN_Light_Attributes(Create Exponent Attribute=true, Create Radius Attribute=true, Exponent=4, Radius=50)
   &InterpolateOverTime(
    Initial Vector Value:NiagaraPosition=$Particles.Initial.Position
    Rate Of Change=0.8
    Target Vector Value:Vector3f=&AddVectorToPosition(Position=$Particles.Initial.Position, Vector=<0 0 37.9136>)
    Variable Type=::NewEnumerator2
   )
   /SetVariables(
    Particles.Flaresize2:Vector2f=&Multiply_Vector2D(A=$Particles.SpriteSize, B=<1.5149 0.25>)
    Particles.Position=$StackContext.InterpolateOverTime.InterpolateOverTime_V3_.Moving Average
   ) {isSetParameters=true}
  ]
  rendererClasses [
   &NiagaraSpriteRendererProperties
   &NiagaraLightRendererProperties
   &NiagaraMeshRendererProperties
  ]
  renderers [
   {rendererClass=&NiagaraSpriteRendererProperties}
   {rendererClass=&NiagaraLightRendererProperties, rendererIndex=1}
   {rendererClass=&NiagaraMeshRendererProperties, rendererIndex=2}
  ]
 }
 SoftSparkles {
  simTarget=GPUComputeSim
  emitterUpdateScript [
   &EmitterState()
   &SpawnRate(SpawnRate=10, Use Spawn Probability=false)
  ]
  particleSpawnScript [
   &InitializeParticle(
    Color=<rgba 0.1 0.0326 0 0.2>
    Lifetime Max=2.4
    Lifetime Min=1.2
    Lifetime Mode=::NewEnumerator1
    Position Offset:NiagaraPosition=$User.PositionOffset
    Position Offset Coordinate Space=::Local
    Sprite Size Mode=::NewEnumerator1
    Uniform Sprite Size Max=44
    Uniform Sprite Size Min=22
    UsePositionOffset=true
    Write Material Random=true
   )
   &SkeletalMeshLocation(
    MaskSpawnGroup=false
    Mesh Sampling Type=::NewEnumerator8
    OffsetAlongNormal=false
    OverrideRandomSeed=false
    Skeletal Mesh=$System.Skeletal Mesh
   )
   /SetVariables(Particles.Curve:NiagaraFloat=&FloatFromCurve(FloatCurve=DataInterface() {propertyValues {Curve {keys [
         {interpMode=RCIM_Cubic, leaveTangent=7.0743, tangentMode=RCTM_Break, time=0, value=0.35}
         {interpMode=RCIM_Cubic, tangentMode=RCTM_Break, time=0.2, value=0.95}
         {interpMode=RCIM_Cubic, leaveTangent=-0.1241, tangentMode=RCTM_Break, time=0.4, value=0.55}
         {interpMode=RCIM_Cubic, tangentMode=RCTM_Break, time=0.6, value=0.8}
         {interpMode=RCIM_Cubic, tangentMode=RCTM_Break, time=0.8, value=0.6}
         {arriveTangent=2.6513, interpMode=RCIM_Cubic, tangentMode=RCTM_Break, time=1, value=1}
        ]}}})) {isSetParameters=true}
  ]
  particleUpdateScript [
   &ParticleState()
   /Niagara/Modules/Update/Forces/V2/CurlNoiseForce(
    Noise Frequency=100
    Noise Strength=&FloatFromCurve(
     FloatCurve=DataInterface() {propertyValues {Curve {keys [{tangentMode=RCTM_User, time=0, value=0}, {tangentMode=RCTM_User, time=1, value=1}]}}}
     Scale Curve=111
    )
    Pan Noise=false
   )
   &SolveForcesAndVelocity(Clamp Velocity=false, Limit Acceleration=false)
   &DynamicMaterialParameters(
    Index 0 Param 1=&Multiply_Float(
     A=&FloatFromCurve(FloatCurve=DataInterface() {propertyValues {Curve {keys [
          {interpMode=RCIM_Cubic, leaveTangent=0.1514, tangentMode=RCTM_Break, time=0, value=1}
          {arriveTangent=-0.3559, interpMode=RCIM_Cubic, tangentMode=RCTM_Break, time=1, value=0}
         ]}}})
     B=&Lerp_Float(A=1, B=0, Alpha=$Particles.Curve)
    )
    Module.Index0Param0Write=true
    Module.Index0Param1Write=true
    Module.Index0Param2Write=true
    Module.Index0Param3Write=true
   )
   &ScaleColor(
    Scale Alpha=&FloatFromCurve(FloatCurve=DataInterface() {propertyValues {Curve {keys [
         {interpMode=RCIM_Cubic, tangentMode=RCTM_User, time=0, value=0}
         {interpMode=RCIM_Cubic, tangentMode=RCTM_User, time=0.5, value=1}
         {tangentMode=RCTM_User, time=1, value=0}
        ]}}})
    Scale RGB=&VectorFromFloat(Value=&Lerp_Float(
      A=1
      B=5
      Alpha=&Sine(Normalized Angle=$Emitter.Age, Period=&RandomRangeFloat(Minimum=0.1, Maximum=0.4), Scale=1, Bias=1)
     ))
    ScaleA=true
    ScaleRGB=true
   )
   &ScaleSpriteSize(
    Scale Sprite Size Mode=::NewEnumerator0
    Uniform Scale Factor=&Multiply_Float(
     A=&FloatFromCurve(FloatCurve=DataInterface() {propertyValues {Curve {keys [
          {interpMode=RCIM_Cubic, leaveTangent=2.1941, tangentMode=RCTM_Break, time=0, value=0.35}
          {arriveTangent=0.204, interpMode=RCIM_Cubic, tangentMode=RCTM_Break, time=1, value=1}
         ]}}})
     B=$Particles.Curve
    )
   )
  ]
  rendererClasses [&NiagaraSpriteRendererProperties]
  renderers [{rendererClass=&NiagaraSpriteRendererProperties}]
 }
 Sparks {
  emitterUpdateScript [
   &EmitterState(Life Cycle Mode=::NewEnumerator1, Loop Behavior=::NewEnumerator1, Loop Duration=1.85, UseLoopDelay=true)
   &SpawnRate(
    SpawnRate=&FloatFromCurve(FloatCurve=DataInterface() {propertyValues {Curve {keys [<at 0.75 1>,<at 1 0>]}}}, CurveIndex=$Emitter.NormalizedLoopAge, Scale Curve=400)
    Use Spawn Probability=true
   )
   &SpawnBurst_Instantaneous(
    Spawn Count=&RandomRangeInt(Minimum=10, Maximum=12)
    Use Spawn Probability=false
    UseLoopCountLimit=false
   ) {enabled=false}
  ]
  particleSpawnScript [
   &InitializeParticle(
    Color=&LerpLinearColors(
     StartColor=<rgba 7 3.6616 1.065 1>
     EndColor=<rgba 8 4.3771 0.7956 1>
     LerpFactor=&MakeCustomFloatFromBool(Bool=&RandomBool(), True Float=1, False Float=0)
    )
    Lifetime Max=0.25
    Lifetime Min=0.1
    Lifetime Mode=::NewEnumerator1
    Sprite Size Mode=::NewEnumerator1
    Uniform Sprite Size Max=2
    Uniform Sprite Size Min=1
    UsePositionOffset=true
    Write Material Random=true
   ) {module=InitializeParticle001}
   &ShapeLocation(
    Cylinder Distribution=::NewEnumerator2
    Cylinder Height=44
    Cylinder Radius=&Multiply_Float(
     A=3.1392
     B=&FloatFromCurve(FloatCurve=DataInterface() {propertyValues {Curve {keys [<at 0.6 1>,<at 2 0.5>]}}}, CurveIndex=$Emitter.Age)
    )
    MaskSpawnGroup=false
    Offset=&MakeVector(X=0, Y=-6.4977, Z=0)
    Offset Coordinate Space=::Local
    Offset Mode=::NewEnumerator0
    Shape Origin=$Engine.Emitter.SimulationPosition
    Shape Primitive=::NewEnumerator1
    U Position=&RandomRangeFloat(Minimum=0, Maximum=1)
    UseOwnerScale=true
    V Position=&FloatFromCurve(FloatCurve=DataInterface() {propertyValues {Curve {keys [<at 0 0.15>,<at 2 1>]}}}, CurveIndex=$Emitter.Age)
   )
   &AddVelocity(
    Velocity=&Lerp_Vector(
     A=Linked() {linkedVariable {name=Particles.ShapeLocation.ShapeNormal, type=NiagaraPosition}}
     B=<0 0 1>
     Alpha=&FloatFromCurve(FloatCurve=DataInterface() {propertyValues {Curve {keys [<at 0.7 0.1>,<at 1 1>]}}}, CurveIndex=$Emitter.NormalizedLoopAge)
    )
    Velocity Speed Scale=20
   )
   &CameraOffset()
  ]
  particleUpdateScript [
   &ParticleState()
   &CurlNoiseForce(
    Noise Frequency=22
    Noise Strength=&Multiply_Float(
     A=100
     B=&FloatFromCurve(FloatCurve=DataInterface() {propertyValues {Curve {keys [<at 0 0>,<at 0.1 1>,<at 1 1>]}}})
    )
    Pan Noise=false
   ) {enabled=false}
   &Drag(Drag=&UniformRangedFloat(Minimum=5, Maximum=10), Use Linear Drag=true, Use Rotational Drag=false) {enabled=false}
   &SolveForcesAndVelocity(Clamp Velocity=false, Limit Acceleration=false)
   &ScaleSpriteSize(Scale Factor=&Vector2DFromFloat(Value=&FloatFromCurve(FloatCurve=DataInterface() {propertyValues {Curve {keys [<at 0 1>,<at 1 0>]}}})))
   &ScaleSpriteSizeBySpeed(Max Scale Factor=<1 10>, Velocity Threshold=250)
   &ScaleColor(
    Scale Alpha=&FloatFromCurve(FloatCurve=DataInterface() {propertyValues {Curve {keys [
         {interpMode=RCIM_Cubic, tangentMode=RCTM_User, time=0, value=0}
         {interpMode=RCIM_Cubic, tangentMode=RCTM_User, time=0.5, value=1}
         {tangentMode=RCTM_User, time=1, value=0}
        ]}}})
    ScaleA=true
    ScaleRGB=false
   )
  ]
  rendererClasses [&NiagaraSpriteRendererProperties]
  renderers [{rendererClass=&NiagaraSpriteRendererProperties}]
 }
}
summary {
 systemName=NS_Backpack_GelatoCourage
 emitters [
  {emitterName=ConfettiStreamers, rendererClasses [&NiagaraMeshRendererProperties]}
  {emitterName=ConfettiSquares, rendererClasses [&NiagaraSpriteRendererProperties]}
  {emitterName=SoftSparkles, simTarget=GPUComputeSim, rendererClasses [&NiagaraSpriteRendererProperties]}
  {emitterName=FlashingSparkles, rendererClasses [&NiagaraSpriteRendererProperties]}
  {
   bEnabled=false
   emitterName=Flare
   rendererClasses [&NiagaraSpriteRendererProperties, &NiagaraLightRendererProperties]
  }
  {emitterName=Sparks, rendererClasses [&NiagaraSpriteRendererProperties]}
  {
   emitterName=Light
   rendererClasses [
    &NiagaraSpriteRendererProperties
    &NiagaraLightRendererProperties
    &NiagaraMeshRendererProperties
   ]
  }
  {
   emitterName=Flare002
   rendererClasses [
    &NiagaraSpriteRendererProperties
    &NiagaraMeshRendererProperties
    &NiagaraLightRendererProperties
   ]
  }
 ]
 userVariables [
  User.bIsFrontEnd {type=NiagaraBool}
  User.bIsFrontEndPreview {type=NiagaraBool}
  User.bReactivity {type=NiagaraBool}
  User.ConfettiColorArray {
   type=NiagaraDataInterfaceArrayColor
   description=""
   defaultValue {
    struct=NiagaraExt_VariableValue_DataInterface
    value {
     dataInterface=&self:NiagaraDataInterfaceArrayColor_0
     dataInterfaceClass=&NiagaraDataInterfaceArrayColor
    }
   }
  }
  User.GoldTransitionCompleteTime {type=NiagaraFloat, defaultValue=1.73, description=""}
  User.PositionOffset {type=NiagaraPosition, description="", defaultValue {}}
 ]
}
