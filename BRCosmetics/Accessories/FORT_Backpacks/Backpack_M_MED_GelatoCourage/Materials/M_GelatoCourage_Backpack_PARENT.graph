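# Parent material for the GelatoCourage backpack that adds a gold victory transition on top of the 2023 character parent functions.
# GoldenAmount_Spyral wipes gold along UV V within the FX texture's blue mask, while GoldenAmount_Base sweeps the remaining metallic areas along pre-skinned Z with panning noise.
# The combined mask tints base color gold.
# A second, offset mask minus the first forms an emissive edge glow, scaled by time of day.
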Material /BRCosmetics/Accessories/FORT_Backpacks/Backpack_M_MED_GelatoCourage/Materials/M_GelatoCourage_Backpack_PARENT
graph {
 properties {
  bUseMaterialAttributes=true
  bUsedWithMeshDeformer=true
  bUsedWithSkeletalMesh=true
  pixelDepthOffsetMode=PDOM_Legacy
  shadingModel=MSM_FromMaterialExpression
  usedShadingModels=Default Lit | Subsurface | Clear Coat | Subsurface Profile | Cloth
 }
 nodes [
  Constant3Vector() {constant=<rgba 1 0.766 0.336 0>}
  n1 = MaterialFunctionCall() {materialFunction=&MF_CharacterParent_2023_PreFX}
  n2 = GetMaterialAttributes(None=MaterialFunctionCall(Material=@n1.Result, M=@n1.Result, Pre Skinned Local Position=@n1.Result, Distance Blend=@n1.Result, Fresnel=@n1.Result) {materialFunction=&MF_CharacterParent_2023_PostFX}.Result) {attributeGetTypes [#1, #2]}
  n3 = Desaturation(None=Multiply(A=Constant3Vector() {constant=<rgba 1 0.766 0.336 0>}) {constB=1.3}, Fraction=Constant() {r=-0.38})
  n4 = ComponentMask(None=TextureCoordinate()) {g=true}
  n5 = LinearInterpolate(Alpha=ScalarParameter() {group=VictoryTransition, parameterName=GoldenAmount_Spyral, sliderMax=1}) {constA=0.1, constAlpha=1, constB=-1}
  n6 = Saturate(None=Multiply(A=Add(A=@n4, B=@n5) {constB=0.1}) {constB=10})
  n7 = TextureSample() {
   samplerType=SAMPLERTYPE_Masks
   texture=/BRCosmetics/Accessories/FORT_Backpacks/Backpack_M_MED_GelatoCourage/Textures/T_GelatoCourage_Backpack_FX
  }
  n8 = ScalarParameter() {group=VictoryTransition, parameterName=GoldenAmount_Base, sliderMax=1}
  n9 = VertexInterpolator(VS=PreSkinnedPosition())
  n10 = ComponentMask(None=Add(A=LinearInterpolate(Alpha=@n8) {constA=20, constAlpha=1, constB=-20}, B=@n9.PS) {constB=14}) {b=true}
  n11 = ComponentMask(None=@n9.PS) {g=true, r=true}
  n12 = ScalarParameter() {defaultValue=0.01, group=VictoryTransition, parameterName=GlowNoiseScale, sortPriority=36}
  n13 = TextureSample(UVs=Panner(Coordinate=Multiply(A=@n11, B=@n12) {constB=0.01}) {speedX=0.1, speedY=-0.07}) {samplerType=SAMPLERTYPE_LinearColor, texture=/Game/Effects/FogVolume/noise_flamy_Tile_Linear}
  n14 = Saturate(None=Add(A=Multiply(A=@n10) {constB=0.2}, B=Multiply(A=@n13.R)))
  n15 = Subtract(A=Multiply(A=@n2.Metallic, B=OneMinus(None=@n14)), B=@n7.B)
  n16 = Add(A=Multiply(A=OneMinus(None=@n6), B=@n7.B), B=Saturate(None=Multiply(A=@n15) {constB=2}))
  n17 = Constant3Vector() {constant=<rgba 1 0.6456 0.0052 1>}
  n18 = Constant() {r=4}
  n19 = MaterialFunctionCall(Value_Morning=@n18, Value_Day=@n18, Value_Evening=Constant() {r=3.2}, Value_Night=Constant() {r=1.75}) {materialFunction=&Scalar_Time_of_Day}
  n20 = ScalarParameter() {group=VictoryTransition, parameterName=GoldenAmount_Spyral, sliderMax=1}
  n21 = ScalarParameter() {defaultValue=0.1, group=VictoryTransition, parameterName=BorderGlowSize, sortPriority=35}
  n22 = Add(A=@n20, B=@n21)
  n23 = Constant()
  n24 = Add(
   A=ComponentMask(None=TextureCoordinate()) {g=true}
   B=LinearInterpolate(Alpha=If(A=@n20, "A > B"=@n22, "A == B"=@n23, "A < B"=@n23)) {constA=0.1, constAlpha=1, constB=-1}
  ) {constB=0.1}
  n25 = OneMinus(None=Saturate(None=Multiply(A=@n24) {constB=10}))
  n26 = ScalarParameter() {group=VictoryTransition, parameterName=GoldenAmount_Base, sliderMax=1}
  n27 = Constant()
  n28 = LinearInterpolate(Alpha=If(A=@n26, "A > B"=Add(A=@n21, B=@n26) {constA=0.1}, "A == B"=@n27, "A < B"=@n27)) {constA=20, constAlpha=1, constB=-20}
  n29 = VertexInterpolator(VS=PreSkinnedPosition())
  n30 = Multiply(A=ComponentMask(None=Add(A=@n28, B=@n29.PS) {constB=14}) {b=true}) {constB=0.2}
  n31 = Multiply(
   A=ComponentMask(None=@n29.PS) {g=true, r=true}
   B=ScalarParameter() {defaultValue=0.01, group=VictoryTransition, parameterName=GlowNoiseScale, sortPriority=36}
  ) {constB=0.01}
  n32 = Multiply(A=TextureSample(UVs=Panner(Coordinate=@n31) {speedX=0.1, speedY=-0.07}) {samplerType=SAMPLERTYPE_LinearColor, texture=&noise_flamy_Tile_Linear}.R)
  n33 = OneMinus(None=Saturate(None=Add(A=@n30, B=@n32)))
  n34 = Multiply(A=Subtract(A=Multiply(A=@n2.Metallic, B=@n33), B=@n7.B)) {constB=2}
  n35 = Subtract(A=Add(A=Multiply(A=@n25, B=@n7.B), B=Saturate(None=@n34)), B=@n16)
  SetMaterialAttributes!(
   MaterialAttributes=@n2.MaterialAttributes
   Base Color=LinearInterpolate(A=@n2.Base Color, B=Multiply(A=@n2.Base Color, B=@n3), Alpha=@n16)
   Emissive Color=Multiply(A=Multiply(A=@n17, B=@n19.Result) {constB=4}, B=Saturate(None=@n35))
  ) {attributeSetTypes [#1, #3]}
 ]
}
